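# Physics asset for the Bane backpack cosmetic, giving its dangling capsules and straps physics bodies relative to the pack.
# Four box bodies cover the pack, a dynamic capsules piece and two strap bones.
# The straps form a chain from the pack through the first to the second segment, while the capsules attach directly to the pack.
# Angular limits are tight, with the upper strap joint allowing the widest swing.
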PhysicsAsset /BRCosmetics/Accessories/FORT_Backpacks/F_MED_Bane_Backpack/Meshes/F_MED_Bane_Pack_Physics
bodies {
 dyn_capsules [{shapeType=Box, extentX=7.766, extentY=2.5243, extentZ=7.9786, center=<0.2563 -0.1585 -0.3771>, rotation=<rot 4.4569 -90 0.3872>}]
 dyn_strap_01 [{shapeType=Box, extentX=9.3143, extentY=1.1709, extentZ=5.8152, center=<2.5868 0.0145 -0.0652>, rotation=<rot 0 0.4635 -89.6954>}]
 dyn_strap_02 [{shapeType=Box, extentX=1.1, extentY=6, extentZ=8, center=<4 0 0>, rotation=<rot 90 0 0>}]
 pack [{shapeType=Box, extentX=23.726, extentY=10, extentZ=20, center=<1.1781 -3.4299 1.9053>}]
}
constraints [
 {bone1Name=dyn_strap_02, bone2Name=dyn_strap_01, swing1LimitDegrees=4, swing2LimitDegrees=10, twistLimitDegrees=5}
 {bone1Name=dyn_strap_01, bone2Name=pack, swing1LimitDegrees=7.5, swing2LimitDegrees=29, twistLimitDegrees=7.5}
 {bone1Name=dyn_capsules, bone2Name=pack, swing1LimitDegrees=5, swing2LimitDegrees=5, twistLimitDegrees=5}
]
